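# Material function for the Flaming Hoops boost ring that overrides emissive, normal and world position offset to animate it.
# Vertex-color masks separate wheel, tank and exhausts, driving wheel rotation, randomized exhaust vibration and a passing color boost.
# Custom primitive data supplies state and Blueprint timestamps for activity timing, with a curve atlas shaping the passing boost.
# Rotation wraps per ring section, using PreviousFrameSwitch to prevent velocity spikes at the wrap.
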MaterialFunction /FlamingHoopsGameplay/Assets/Gameplay_BoostRing_A/Materials/MF_FlamingHoops
graph {nodes [
  n1 = GetMaterialAttributes(None=FunctionInput() {inputType=FunctionInput_MaterialAttributes}) {attributeGetTypes [#1, #2, #3]}
  n2 = SetMaterialAttributes!(
   MaterialAttributes=@n1.MaterialAttributes
   EmissiveColor=NamedRerouteUsage()
   Normal=NamedRerouteUsage()
   WorldPositionOffset=NamedRerouteUsage()
   BaseColor
  ) {attributeSetTypes [
    #3
    #1
    #4
    #2
   ]}
  FunctionOutput(None=@n2)
  n3 = ScalarParameter() {defaultValue=0.1, desc=0.2, group=FlamingHoops, parameterName=FH_RingActiveSpeed}
  n4 = NamedRerouteUsage()
  n5 = Add(A=@n4, B=NamedRerouteUsage())
  n6 = NamedRerouteUsage()
  n7 = CurveAtlasRowParameter(CurveTime=@n6) {
   atlas=/FlamingHoopsGameplay/Assets/Gameplay_BoostRing_A/Textures/CA_FlamingHoops
   curve=/FlamingHoopsGameplay/Assets/Gameplay_BoostRing_A/Textures/C_FlamingHoops
   group=FlamingHoops
   parameterName=FH_PassingCurveAnim
  }
  n8 = LinearInterpolate(A=@n4, B=@n5, Alpha=@n7.R)
  n9 = Add(A=NamedRerouteUsage(), B=ScalarParameter() {group=FlamingHoops, parameterName=FH_ActiveRotationOffset}) {constB=0.5}
  n10 = CurveAtlasRowParameter(CurveTime=@n6) {
   atlas=&CA_FlamingHoops
   curve=&C_FlamingHoops
   group=FlamingHoops
   parameterName=FH_PassingCurveAnim
  }
  n11 = LinearInterpolate(A=Add(A=Frac(None=@n5), B=Floor(None=@n9)), B=Time(), Alpha=@n10.R)
  If(
   A=NamedRerouteUsage()
   "A > B"=Frac(None=Multiply(A=@n8, B=@n3) {constB=0.1, desc="distance = time * speed"}) {desc=PreCooldown/slowing down}
   "A == B"=Frac(None=Multiply(A=@n3, B=@n11)) {desc=Not spinning}
   "A < B"=Frac(None=Multiply(A=@n5, B=@n3)) {desc=Not spinning}
  )
  n12 = Time()
  n13 = ScalarParameter() {defaultValue=3, group=FlamingHoops, parameterName=FH_Exhaust_VibrateFrequency}
  n14 = NamedRerouteUsage()
  n15 = ScalarParameter() {defaultValue=1, group=FlamingHoops, parameterName=FH_RandomMask_Intensity}
  n16 = Sine(None=Subtract(A=Multiply(A=@n12, B=@n13) {constB=7}, B=Multiply(A=@n14, B=@n15) {constB=5}))
  MaterialExpressionTime_6 = Time()
  MaterialExpressionScalarParameter_11 = ScalarParameter() {defaultValue=5, group=FlamingHoops, parameterName=FH_FlickerSpeed}
  n17 = Custom!(time=@MaterialExpressionTime_6, speed=@MaterialExpressionScalarParameter_11) {
   code="float x = sin(time*speed*1.9898)/2+0.5;\r\nfloat y = cos(time*speed*0.3753)/2+0.5;\r\nfloat z = sin(time*speed*1.2367)/2+0.5;\r\nreturn float3(x,y,z);"
   description=flicker
   inputs [
    {inputName=time, input {expression=&self:MaterialExpressionTime_6}}
    {
     inputName=speed
     input {expression=&self:MaterialExpressionScalarParameter_11}
    }
   ]
  }
  n18 = VectorParameter() {desc=TO-DO: RANDOM, group=FlamingHoops, parameterName=FH_Exhaust_VibrateAmp, defaultValue=<rgba 5 5 5 1>}
  n19 = Multiply(A=Normalize(VectorInput=@n17), B=@n18.RGB)
  n20 = Multiply(
   A=Multiply(A=@n16, B=Multiply(A=@n16, B=@n19))
   B=ScalarParameter() {defaultValue=0.3, group=FlamingHoops, parameterName=FH_ExhaustVibratentensity}
  )
  n21 = NamedRerouteUsage()
  n22 = NamedRerouteUsage()
  LinearInterpolate(B=Add(B=@n20), Alpha=LinearInterpolate(A=NamedRerouteUsage(), B=OneMinus(None=@n21), Alpha=Saturate(None=@n22)))
  MaterialFunctionCall() {materialFunction=&ObjectPivotPoint}
  n23 = NamedRerouteUsage()
  n24 = Multiply(A=@n23, B=VectorParameter() {group=FlamingHoops, parameterName=FH_Ring_PassingColorBoost, defaultValue=<rgba 50 6.8332 0 1>}.RGB)
  n25 = OneMinus(None=NamedRerouteUsage())
  NamedRerouteDeclaration(None=NamedRerouteDeclaration(None=LinearInterpolate(A=@n23, B=@n24, Alpha=@n25)) {name=Passing_Color, nodeColor=<rgba 1 0.8438 0 1>}) {name=Emissive_OUT_Wheel, nodeColor=<rgba 0 0.3672 1 1>}
  NamedRerouteDeclaration(None=Multiply(A=@n20, B=NamedRerouteUsage())) {name=WPO_OUT_Exhausts, nodeColor=<rgba 0 0.3672 1 1>}
  NamedRerouteDeclaration(None=@n1.Normal) {name=Normal_IN, nodeColor=<rgba 0 0.5312 1 1>}
  n26 = VertexColor()
  NamedRerouteDeclaration(None=Step(X=@n26.R) {constX=0.01, constY=0.99}) {name=Mask_Wheel, nodeColor=<rgba 1 0.0127 0 1>}
  NamedRerouteDeclaration(None=Step(X=@n26.G) {constX=0.01, constY=0.99}) {name=Mask_Tank, nodeColor=<rgba 0 1 0.0137 1>}
  n27 = NamedRerouteUsage()
  n28 = NamedRerouteUsage()
  n29 = NamedRerouteUsage()
  n30 = NamedRerouteUsage()
  NamedRerouteDeclaration(None=LinearInterpolate(A=LinearInterpolate(B=@n27, Alpha=@n28), B=NamedRerouteUsage(), Alpha=Add(A=@n29, B=@n30))) {name=WPO_OUT, nodeColor=<rgba 1 0.5859 0 1>}
  NamedRerouteDeclaration(None=LinearInterpolate(A=NamedRerouteUsage(), B=NamedRerouteUsage(), Alpha=NamedRerouteUsage())) {name=Normal_OUT, nodeColor=<rgba 0.7812 1 0 1>}
  NamedRerouteDeclaration(None=NamedRerouteUsage()) {name=Emissive_OUT, nodeColor=<rgba 0 0.8594 1 1>}
  NamedRerouteDeclaration(None=@n1.BaseColor) {name=BaseColor_IN, nodeColor=<rgba 0.1953 0 1 1>}
  NamedRerouteDeclaration(None=@n1.EmissiveColor) {name=Emissive_IN, nodeColor=<rgba 0 0.3672 1 1>}
  n31 = Normalize(VectorInput=VectorParameter() {group=FlamingHoops, parameterName=FH_WheelRotateAxis, defaultValue=<rgba 1 0 0 1>}.RGB)
  n32 = NamedRerouteUsage()
  n33 = Transform(None=NamedRerouteUsage()) {transformType=TRANSFORM_Local}
  n34 = Constant()
  NamedRerouteDeclaration(None=Transform(None=MaterialFunctionCall(Rotation Axis=@n31, Rotation Angle=@n32, World Space Vertex Normals=@n33, Normals Pivot=@n34) {materialFunction=&FixRotateAboutAxisNormals}.New World Space Vertex Normal) {transformSourceType=TRANSFORMSOURCE_Local, transformType=TRANSFORM_Tangent}) {name=Normal_OUT_Wheel, nodeColor=<rgba 0 0.1562 1 1>}
  n35 = Step(X=@n26.B) {constX=0.01, constY=0.01}
  NamedRerouteDeclaration(None=Multiply(A=NamedRerouteDeclaration(None=@n35) {name=Mask_Exhausts, nodeColor=<rgba 0.0099 0 1 1>}, B=@n26.G)) {name=RandomMask_Exhausts, nodeColor=<rgba 0.5 1 0 1>}
  n36 = LocalPosition() {localOrigin=InstancePreSkinning}
  NamedRerouteDeclaration(None=Transform(None=RotateAboutAxis(NormalizedRotationAxis=@n31, RotationAngle=@n32, PivotPoint=@n34, Position=@n36.XYZ)) {transformSourceType=TRANSFORMSOURCE_Local}) {name=WPO_OUT_Wheel, nodeColor=<rgba 0 0.5547 1 1>}
  n37 = ScalarParameter() {bUseCustomPrimitiveData=true, group=FlamingHoops, parameterName=FH_State}
  n38 = ScalarParameter() {bUseCustomPrimitiveData=true, group=FlamingHoops, parameterName=GameTimeFromBP_Passing, primitiveDataIndex=2}
  n39 = Divide(
   A=Subtract(A=Time(), B=NamedRerouteDeclaration(None=@n38) {name=GameTime_Passing, nodeColor=<rgba 0.0547 0 1 1>})
   B=NamedRerouteDeclaration(None=ScalarParameter() {defaultValue=5, group=FlamingHoops, parameterName=FH_Passing_Duration, primitiveDataIndex=2}) {name=PreCooldownDuration, nodeColor=<rgba 0 1 0.5313 1>}
  )
  n40 = Time()
  n41 = NamedRerouteDeclaration(None=ScalarParameter() {bUseCustomPrimitiveData=true, group=FlamingHoops, parameterName=GameTimeFromBP_Active, primitiveDataIndex=1}) {name=GameTime_Active, nodeColor=<rgba 0.0547 0 1 1>}
  n42 = Divide(
   A=NamedRerouteDeclaration(None=Subtract(A=@n40, B=@n41)) {name=TimePassed_Active, nodeColor=<rgba 0.9219 0 1 1>}
   B=NamedRerouteDeclaration(None=ScalarParameter() {defaultValue=5, group=FlamingHoops, parameterName=FH_Speedup_Duration, primitiveDataIndex=2}) {name=SpeedUpDuration, nodeColor=<rgba 0 1 0.5312 1>}
  )
  NamedRerouteDeclaration(None=If(
    A=NamedRerouteDeclaration(None=@n37) {name=State, nodeColor=<rgba 0 0.3906 1 1>}
    "A > B"=Saturate(None=@n39)
    "A == B"=Saturate(None=@n42)
    "A < B"=Constant()
   )) {name=Activity, nodeColor=<rgba 0 1 0.8594 1>}
  n43 = Multiply(A=Time(), B=@n3)
  n44 = CurveAtlasRowParameter(CurveTime=NamedRerouteUsage()) {
   atlas=&CA_FlamingHoops
   curve=&C_FlamingHoops
   group=FlamingHoops
   parameterName=FH_PassingCurveAnim
  }
  n45 = Multiply(
   A=NamedRerouteDeclaration(None=@n44.B) {name=PassingBoostCurve, nodeColor=<rgba 1 0 0.0703 1>}
   B=ScalarParameter() {defaultValue=0.55, group=FlamingHoops, parameterName=FH_CurveMulti}
  ) {constB=0.5}
  n46 = Add(A=@n43, B=@n45)
  n47 = Constant() {r=360}
  n48 = Divide(A=Divide(A=@n47, B=ScalarParameter() {defaultValue=30, parameterName=NumSections}), B=@n47)
  n49 = Multiply(A=DeltaTime(), B=@n3)
  n50 = Modulo!(A=Add(A=Add(A=@n43, B=@n49), B=@n45), B=@n48)
  NamedRerouteDeclaration(None=PreviousFrameSwitch(Current Frame=Modulo(A=@n46, B=@n48), Previous Frame=Subtract(A=@n50, B=@n49)) {desc=Without this the velocity is insane on the jump back}) {name=RotationAngle, nodeColor=<rgba 0 1 0.6016 1>}
  NamedRerouteUsage()
  ScalarParameter() {defaultValue=1, group=FlamingHoops, parameterName=State}
  VertexInterpolator(VS=PreSkinnedNormal())
  WorldPosition()
 ]}
